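# Material function that varies a color with a noise-driven hue tint and darkening.
# It multiplies OriginalColor by HueTintColor, then lerps from the original to that tinted result using the scalar HueNoise.
# The blend is multiplied by the scalar DarkenedNoise to produce ModifiedColor.
# A HueShift call on the original color is unconnected to the output, so the hue change comes only from the tint multiply.
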MaterialFunction /PCGPrimitives_MeshPartitionInterop/Assets/Materials/MF_ApplyHueShiftDarken
graph {nodes [
  n1 = FunctionInput() {inputName=OriginalColor, previewValue=<1 1 1 1>}
  n2 = Multiply(A=@n1, B=FunctionInput() {inputName=HueTintColor, previewValue=<1 1 1 1>})
  n3 = FunctionInput() {inputName=HueNoise, inputType=FunctionInput_Scalar, sortPriority=10, previewValue=<1 0 0 1>}
  n4 = FunctionInput() {inputName=DarkenedNoise, inputType=FunctionInput_Scalar, sortPriority=20, previewValue=<1 0 0 1>}
  FunctionOutput(None=Multiply(A=LinearInterpolate(A=@n1, B=@n2, Alpha=@n3), B=@n4) {constB=0.5}) {outputName=ModifiedColor}
  MaterialFunctionCall(Hue Shift Percentage=Constant() {r=-0.05}, Texture=@n1) {materialFunction=&HueShift}
 ]}
