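# Material function that outputs HueNoise and DarkenNoise, large-scale world-space variation values for hue and darkening.
# Each output samples a texture's green channel using world XY, excluding shader offsets, divided by a large constant; DarkenNoise remaps its sample through a lerp.
# Procedural noise products, including Voronoi, remain alongside feature-level switches and debug-toggle lerps, but none of these reach the outputs.
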MaterialFunction /PCGPrimitives_MeshPartitionInterop/Assets/Materials/MF_ComputeHueAndDarkenNoise
graph {nodes [
  n1 = WorldPosition()
  n2 = Multiply(
   A=Noise!(World Position=@n1.XYZ, FilterWidth=Constant() {r=0}) {outputMin=0.6, scale=0}
   B=Noise(World Position=@n1.XYZ, FilterWidth=Constant() {r=0.0001}) {outputMin=0.6, scale=0.0001}
  )
  n3 = WorldPosition() {worldPositionShaderOffset=WPT_ExcludeAllShaderOffsets}
  n4 = LinearInterpolate(Alpha=TextureSample(UVs=Divide(A=@n3.XY) {constB=-35000}).G) {constA=0.6, constB=1.1}
  FeatureLevelSwitch(Default=@n4, SM5=@n2, SM6=@n2)
  n5 = Constant() {r=0.0001}
  n6 = Multiply(
   A=Noise(World Position=@n1.XYZ, FilterWidth=@n5) {outputMin=0, scale=0.0001}
   B=Noise(World Position=@n1.XYZ, FilterWidth=@n5) {noiseFunction=NOISEFUNCTION_VoronoiALU, outputMin=0, scale=0.0001}
  ) {constB=0.1}
  n7 = TextureSample(UVs=Divide(A=@n3.XY) {constB=37599})
  FeatureLevelSwitch(Default=@n7.G, SM5=@n6, SM6=@n6)
  FunctionOutput(None=@n7.G) {outputName=HueNoise}
  FunctionOutput(None=@n4) {outputName=DarkenNoise}
  n8 = MaterialFunctionCall() {materialFunction=/Engine/Functions/Engine_MaterialFunctions02/Utility/DebugOnOff}
  LinearInterpolate(A=@n6, B=@n7.G, Alpha=@n8.Result) {constB=0}
  LinearInterpolate(A=@n2, B=@n4, Alpha=@n8.Result) {constAlpha=1}
 ]}
