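# Material function that projects a PBR texture set (color, normal, ORM, height) in world space and outputs combined Material Attributes.
# Each texture input runs through WorldAlignedTexture's XYZ output, with ORM split into AO, roughness and metallic and specular supplied as a scalar.
# The projected normal has X and Y negated.
# A shared Texture Size input sets the projection scale, except for the height-to-displacement projection, which leaves its size unconnected.
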MaterialFunction /PCGPrimitives_MeshPartitionInterop/Assets/Materials/MF_WorldAlignedTextureSet
graph {nodes [
  FunctionOutput(None=MakeMaterialAttributes!(
    BaseColor=NamedRerouteUsage()
    Metallic=NamedRerouteUsage()
    Specular=NamedRerouteUsage()
    Roughness=NamedRerouteUsage()
    Anisotropy
    EmissiveColor
    Opacity
    OpacityMask
    Normal=NamedRerouteUsage()
    Tangent
    WorldPositionOffset
    SubsurfaceColor
    ClearCoat
    ClearCoatRoughness
    AmbientOcclusion=NamedRerouteUsage()
    Refraction
    CustomizedUVs_0
    CustomizedUVs_1
    CustomizedUVs_2
    CustomizedUVs_3
    CustomizedUVs_4
    CustomizedUVs_5
    CustomizedUVs_6
    CustomizedUVs_7
    PixelDepthOffset
    ShadingModel
    Displacement=NamedRerouteUsage()
   ))
  n1 = FunctionInput(Preview=TextureObject() {samplerType=SAMPLERTYPE_Normal, texture=/Fab/Textures/Standard/T_DefaultNormal}) {bUsePreviewValueAsDefault=true, inputName=Texture N, inputType=FunctionInput_Texture2D, sortPriority=1}
  n2 = NamedRerouteUsage()
  NamedRerouteDeclaration(None=Multiply(
    A=MaterialFunctionCall(TextureObject=@n1, TextureSize=@n2) {materialFunction=&WorldAlignedTexture}.XYZ Texture
    B=Constant3Vector() {constant=<rgba -1 -1 1 0>}
   ) {desc=Flip normal}) {name=Normal, nodeColor=<rgba 0 1 0.4609 1>}
  n3 = TextureObject() {samplerType=SAMPLERTYPE_Alpha, texture=/Fab/Textures/Standard/T_DefaultDisplacement}
  NamedRerouteDeclaration(None=MaterialFunctionCall(TextureObject=FunctionInput(Preview=@n3) {bUsePreviewValueAsDefault=true, inputName=Texture H, inputType=FunctionInput_Texture2D, sortPriority=3}) {materialFunction=&WorldAlignedTexture}.XYZ Texture) {name=Displacement, nodeColor=<rgba 0 1 0.6484 1>}
  MaterialExpressionMaterialFunctionCall_3 = MaterialFunctionCall(
   TextureObject=FunctionInput(Preview=TextureObject() {samplerType=SAMPLERTYPE_Masks, texture=/Fab/Textures/Standard/T_DefaultMasks}) {bUsePreviewValueAsDefault=true, inputName=Texture ORM, inputType=FunctionInput_Texture2D, sortPriority=2}
   TextureSize=NamedRerouteUsage()
  ) {materialFunction=&WorldAlignedTexture}
  n4 = Convert(None=@MaterialExpressionMaterialFunctionCall_3.XYZ Texture) {
   nodeName=Break Float3
   convertInputs [{
     type=Vector3
     expressionInput {
      expression=&self:MaterialExpressionMaterialFunctionCall_3
      inputName=None
      mask=0
      maskA=0
      maskB=0
      maskG=0
      maskR=0
      outputIndex=2
     }
    }]
   convertOutputs [{type=Scalar}, {type=Scalar}, {type=Scalar}]
  }
  NamedRerouteDeclaration(None=@n4) {name=AO, nodeColor=<rgba 0 1 0.8828 1>}
  NamedRerouteDeclaration(None=FunctionInput() {inputName=Texture Size, sortPriority=5}) {name=Texture Size, nodeColor=<rgba 1 0.7969 0 1>}
  NamedRerouteDeclaration(None=FunctionInput(Preview=Constant() {r=0.5}) {inputName=Specular, inputType=FunctionInput_Scalar, sortPriority=4, previewValue=<0.5 0 0 1>}) {name=Specular, nodeColor=<rgba 0.6172 1 0 1>}
  NamedRerouteDeclaration(None=@n4) {name=Metallic, nodeColor=<rgba 1 0.5625 0 1>}
  n5 = TextureObject() {texture=/Fab/Textures/Standard/T_DefaultColor}
  NamedRerouteDeclaration(None=MaterialFunctionCall(
    TextureObject=FunctionInput(Preview=@n5) {bUsePreviewValueAsDefault=true, inputName=Texture D, inputType=FunctionInput_Texture2D}
    TextureSize=NamedRerouteUsage()
   ) {materialFunction=&WorldAlignedTexture}.XYZ Texture) {name=Base Color, nodeColor=<rgba 0.125 0 1 1>}
  NamedRerouteDeclaration(None=@n4) {name=Roughness, nodeColor=<rgba 0.8281 1 0 1>}
 ]}
