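# Material function that builds a world-aligned grid surface as a full material attributes output, with no inputs, from the engine's default grid textures.
# World Aligned Texture samples the grid mask at three world scales, combining its channels into gray base color and roughness variation.
# Camera Depth Fade limits small-tile detail in base color by distance and also modulates roughness.
# A finer world-aligned grid normal is sign-corrected and scaled; specular and displacement are constants.
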MaterialFunction /PCGPrimitives_MeshPartitionInterop/Assets/Materials/MF_WorldAlignedGrid
graph {nodes [
  n1 = TextureObject() {samplerType=SAMPLERTYPE_LinearColor, texture=/Engine/EngineMaterials/T_Default_Material_Grid_M}
  n2 = Constant3Vector() {constant=<rgba 200 200 200 0>}
  n3 = Divide(A=@n2) {constB=25}
  n4 = LinearInterpolate(Alpha=ComponentMask(None=MaterialFunctionCall(TextureObject=@n1, TextureSize=@n3) {materialFunction=&WorldAlignedTexture}.XYZ Texture) {r=true}) {constA=0.5}
  n5 = Divide(A=@n2) {constB=10}
  n6 = ComponentMask(None=MaterialFunctionCall(TextureObject=@n1, TextureSize=@n5) {materialFunction=&WorldAlignedTexture}.XYZ Texture) {g=true}
  n7 = MaterialFunctionCall(Fade Length=Constant() {r=1500}) {materialFunction=&CameraDepthFade}
  n8 = LinearInterpolate(A=@n4, B=OneMinus(None=@n4), Alpha=LinearInterpolate(A=@n6, Alpha=@n7.Result))
  n9 = ComponentMask(None=MaterialFunctionCall(TextureObject=@n1, TextureSize=@n2) {materialFunction=&WorldAlignedTexture}.XYZ Texture) {g=true}
  n10 = ComponentMask(None=Add(A=@n9, B=LinearInterpolate(A=@n6, Alpha=@n7.Result) {constB=0, desc=Add small tiles to basecolor near camera}) {constB=0}) {r=true}
  n11 = Constant() {r=0.5}
  n12 = Multiply(
   A=LinearInterpolate(A=Constant() {r=0.295}, B=Constant() {r=0.66}, Alpha=LinearInterpolate(A=@n10, Alpha=@n11) {constAlpha=0.875, constB=0.5})
   B=Constant() {r=0.5}
  ) {constB=0.5}
  n13 = LinearInterpolate(Alpha=@n6) {constA=0.7}
  n14 = Add(A=LinearInterpolate(Alpha=@n9) {constB=0.25}, B=Multiply(A=LinearInterpolate(A=@n13, Alpha=@n7.Result)))
  n15 = Multiply(
   A=MaterialFunctionCall(
    TextureObject=TextureObject() {samplerType=SAMPLERTYPE_Normal, texture=/Engine/EngineMaterials/T_Default_Material_Grid_N}
    TextureSize=Divide(A=@n5)
   ) {materialFunction=&WorldAlignedTexture}.XYZ Texture
   B=Constant3Vector() {constant=<rgba -1 -1 1 0>}
  )
  n16 = Constant3Vector() {constant=<rgba 0.3 0.3 1 0>}
  FunctionOutput(None=MakeMaterialAttributes!(
    BaseColor=Multiply(A=@n8, B=@n12)
    Metallic
    Specular=Constant() {r=0.5}
    Roughness=Clamp(None=@n14)
    Anisotropy
    EmissiveColor
    Opacity
    OpacityMask
    Normal=Multiply(A=@n15, B=@n16)
    Tangent
    WorldPositionOffset
    SubsurfaceColor
    ClearCoat
    ClearCoatRoughness
    AmbientOcclusion
    Refraction
    CustomizedUVs_0
    CustomizedUVs_1
    CustomizedUVs_2
    CustomizedUVs_3
    CustomizedUVs_4
    CustomizedUVs_5
    CustomizedUVs_6
    CustomizedUVs_7
    PixelDepthOffset
    ShadingModel
    Displacement=Constant() {r=0.5}
   ))
  StaticBool() {value=true}
 ]}
