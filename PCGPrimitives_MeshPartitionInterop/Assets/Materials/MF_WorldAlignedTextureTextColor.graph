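# Material function that builds a matte material showing a tinted, world-aligned text texture, such as the default "Grass" text texture.
# WorldAlignedTexture projects the Texture Text input at Texture Size, and its XYZ output is multiplied by the Color input, white by default, to form Base Color.
# It outputs full Material Attributes with constant metallic, specular and roughness, leaving normal and all other attributes unconnected.
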MaterialFunction /PCGPrimitives_MeshPartitionInterop/Assets/Materials/MF_WorldAlignedTextureTextColor
graph {nodes [
  FunctionOutput(None=MakeMaterialAttributes!(
    BaseColor=NamedRerouteUsage()
    Metallic=Constant()
    Specular=Constant() {r=0.5}
    Roughness=Constant() {r=1}
    Anisotropy
    EmissiveColor
    Opacity
    OpacityMask
    Normal
    Tangent
    WorldPositionOffset
    SubsurfaceColor
    ClearCoat
    ClearCoatRoughness
    AmbientOcclusion
    Refraction
    CustomizedUVs_0
    CustomizedUVs_1
    CustomizedUVs_2
    CustomizedUVs_3
    CustomizedUVs_4
    CustomizedUVs_5
    CustomizedUVs_6
    CustomizedUVs_7
    PixelDepthOffset
    ShadingModel
    Displacement
   ))
  NamedRerouteDeclaration(None=FunctionInput() {inputName=Texture Size, sortPriority=5}) {name=Texture Size, nodeColor=<rgba 1 0.7969 0 1>}
  n1 = FunctionInput(Preview=TextureObject() {texture=/PCGPrimitives_MeshPartitionInterop/Assets/Textures/Text/Text_Grass}) {bUsePreviewValueAsDefault=true, inputName=Texture Text, inputType=FunctionInput_Texture2D}
  n2 = NamedRerouteUsage()
  n3 = Constant3Vector() {constant=<rgba 1 1 1 0>}
  NamedRerouteDeclaration(None=Multiply(
    A=MaterialFunctionCall(TextureObject=@n1, TextureSize=@n2) {materialFunction=&WorldAlignedTexture}.XYZ Texture
    B=FunctionInput(Preview=@n3) {bUsePreviewValueAsDefault=true, inputName=Color}
   )) {name=Base Color, nodeColor=<rgba 0.125 0 1 1>}
 ]}
