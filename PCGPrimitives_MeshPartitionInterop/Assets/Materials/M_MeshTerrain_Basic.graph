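# M_MeshTerrain_Basic is a Nanite terrain material for Mesh Partition that displays painted surface layers as colored, world-aligned name-label textures.
# Channel samples from MPD_Basic supply weights for Grass, Dirt, ForestFloor, Mud, Gravel and Rock.
# Chained per-layer composite subgraphs start from a world-aligned grid and blend each layer's label and flat color over the previous result by weight, using material attributes.
# Tessellation and displacement fade are enabled.
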Material /PCGPrimitives_MeshPartitionInterop/Assets/Materials/M_MeshTerrain_Basic
graph {
 expressions {
  MaterialExpressionBlendMaterialAttributes_0 {
   class=MaterialExpressionBlendMaterialAttributes
   inputs [
    A {from=MaterialExpressionReroute_0, output=A}
    B {from=MaterialExpressionMaterialFunctionCall_7, output=Result}
    Alpha {from=MaterialExpressionNamedRerouteUsage_3, output=""}
   ]
   properties {}
  }
  MaterialExpressionBlendMaterialAttributes_1 {
   class=MaterialExpressionBlendMaterialAttributes
   inputs [
    A {from=MaterialExpressionReroute_0, output=A}
    B {from=MaterialExpressionMaterialFunctionCall_8, output=Result}
    Alpha {from=MaterialExpressionNamedRerouteUsage_8, output=""}
   ]
   properties {}
  }
  MaterialExpressionBlendMaterialAttributes_2 {
   class=MaterialExpressionBlendMaterialAttributes
   inputs [
    A {from=MaterialExpressionReroute_8, output=A}
    B {from=MaterialExpressionMaterialFunctionCall_9, output=Result}
    Alpha {from=MaterialExpressionNamedRerouteUsage_10, output=""}
   ]
   properties {}
  }
  MaterialExpressionBlendMaterialAttributes_3 {
   class=MaterialExpressionBlendMaterialAttributes
   inputs [
    A {from=MaterialExpressionReroute_4, output=A}
    B {from=MaterialExpressionMaterialFunctionCall_10, output=Result}
    Alpha {from=MaterialExpressionNamedRerouteUsage_20, output=""}
   ]
   properties {}
  }
  MaterialExpressionBlendMaterialAttributes_4 {
   class=MaterialExpressionBlendMaterialAttributes
   inputs [
    A {from=MaterialExpressionReroute_8, output=A}
    B {from=MaterialExpressionMaterialFunctionCall_11, output=Result}
    Alpha {from=MaterialExpressionNamedRerouteUsage_21, output=""}
   ]
   properties {}
  }
  MaterialExpressionBlendMaterialAttributes_5 {
   class=MaterialExpressionBlendMaterialAttributes
   inputs [
    A {from=MaterialExpressionReroute_4, output=A}
    B {from=MaterialExpressionMaterialFunctionCall_12, output=Result}
    Alpha {from=MaterialExpressionNamedRerouteUsage_22, output=""}
   ]
   properties {}
  }
  MaterialExpressionComposite_0 {class=MaterialExpressionComposite, inputs [A {from=MaterialExpressionReroute_3, output=Output}], properties {subgraphName=Dirt}}
  MaterialExpressionComposite_1 {
   class=MaterialExpressionComposite
   inputs [A {from=MaterialExpressionMaterialFunctionCall_6, output=Result}]
   properties {subgraphName=Grass}
  }
  MaterialExpressionComposite_2 {class=MaterialExpressionComposite, inputs [A {from=MaterialExpressionReroute_1, output=Output}], properties {subgraphName=ForestFloor}}
  MaterialExpressionComposite_3 {class=MaterialExpressionComposite, inputs [A {from=MaterialExpressionReroute_9, output=Output}], properties {subgraphName=Mud}}
  MaterialExpressionComposite_4 {class=MaterialExpressionComposite, inputs [A {from=MaterialExpressionReroute_5, output=Output}], properties {subgraphName=Gravel}}
  MaterialExpressionComposite_5 {class=MaterialExpressionComposite, inputs [A {from=MaterialExpressionReroute_9, output=Output}], properties {subgraphName=Rock}}
  MaterialExpressionConstant3Vector_0 {class=MaterialExpressionConstant3Vector, properties {constant=<rgba 100 100 100 0>}}
  MaterialExpressionConstant3Vector_2 {class=MaterialExpressionConstant3Vector, properties {constant=<rgba 0.0848 0.3333 0.0642 1>}}
  MaterialExpressionConstant3Vector_3 {class=MaterialExpressionConstant3Vector, properties {constant=<rgba 0.2656 0.098 0.0579 1>}}
  MaterialExpressionConstant3Vector_4 {class=MaterialExpressionConstant3Vector, properties {constant=<rgba 0.2865 0.1718 0.085 1>}}
  MaterialExpressionConstant3Vector_5 {class=MaterialExpressionConstant3Vector, properties {constant=<rgba 0.099 0.0415 0.0304 1>}}
  MaterialExpressionConstant3Vector_6 {class=MaterialExpressionConstant3Vector, properties {constant=<rgba 0.2031 0.1507 0.127 1>}}
  MaterialExpressionConstant3Vector_7 {class=MaterialExpressionConstant3Vector, properties {constant=<rgba 0.0938 0.083 0.0786 1>}}
  MaterialExpressionMaterialFunctionCall_10 {
   class=MaterialExpressionMaterialFunctionCall
   inputs [
    Texture Text {from=MaterialExpressionTextureObject_26}
    Color {from=MaterialExpressionConstant3Vector_5}
    Texture Size {from=MaterialExpressionNamedRerouteUsage_5}
   ]
   properties {materialFunction=&/MF_WorldAlignedTextureTextColor}
  }
  MaterialExpressionMaterialFunctionCall_11 {
   class=MaterialExpressionMaterialFunctionCall
   inputs [
    Texture Text {from=MaterialExpressionTextureObject_27}
    Color {from=MaterialExpressionConstant3Vector_6}
    Texture Size {from=MaterialExpressionNamedRerouteUsage_7}
   ]
   properties {materialFunction=&MF_WorldAlignedTextureTextColor}
  }
  MaterialExpressionMaterialFunctionCall_12 {
   class=MaterialExpressionMaterialFunctionCall
   inputs [
    Texture Text {from=MaterialExpressionTextureObject_28}
    Color {from=MaterialExpressionConstant3Vector_7}
    Texture Size {from=MaterialExpressionNamedRerouteUsage_9}
   ]
   properties {materialFunction=&MF_WorldAlignedTextureTextColor}
  }
  MaterialExpressionMaterialFunctionCall_6 {
   class=MaterialExpressionMaterialFunctionCall
   inputs []
   properties {materialFunction=&/MF_WorldAlignedGrid}
  }
  MaterialExpressionMaterialFunctionCall_7 {
   class=MaterialExpressionMaterialFunctionCall
   inputs [
    Texture Text {from=MaterialExpressionTextureObject_0}
    Color {from=MaterialExpressionConstant3Vector_2}
    Texture Size {from=MaterialExpressionNamedRerouteUsage_4}
   ]
   properties {materialFunction=&MF_WorldAlignedTextureTextColor}
  }
  MaterialExpressionMaterialFunctionCall_8 {
   class=MaterialExpressionMaterialFunctionCall
   inputs [
    Texture Text {from=MaterialExpressionTextureObject_24}
    Color {from=MaterialExpressionConstant3Vector_3}
    Texture Size {from=MaterialExpressionNamedRerouteUsage_0}
   ]
   properties {materialFunction=&MF_WorldAlignedTextureTextColor}
  }
  MaterialExpressionMaterialFunctionCall_9 {
   class=MaterialExpressionMaterialFunctionCall
   inputs [
    Texture Text {from=MaterialExpressionTextureObject_25}
    Color {from=MaterialExpressionConstant3Vector_4}
    Texture Size {from=MaterialExpressionNamedRerouteUsage_1}
   ]
   properties {materialFunction=&MF_WorldAlignedTextureTextColor}
  }
  MaterialExpressionMeshPartitionChannelSample_0 {
   class=MaterialExpressionMeshPartitionChannelSample
   inputs [ChannelTextureInput {from=MaterialExpressionMeshPartitionResource_0, output=Channel Texture}]
   properties {channel=Grass, meshPartitionDefinition=/PCGPrimitives_MeshPartitionInterop/Assets/MeshPartitionDefinitions/MPD_Basic}
  }
  MaterialExpressionMeshPartitionChannelSample_1 {
   class=MaterialExpressionMeshPartitionChannelSample
   inputs [ChannelTextureInput {from=MaterialExpressionMeshPartitionResource_0, output=Channel Texture}]
   properties {channel=Dirt, meshPartitionDefinition=&MPD_Basic}
  }
  MaterialExpressionMeshPartitionChannelSample_2 {
   class=MaterialExpressionMeshPartitionChannelSample
   inputs [ChannelTextureInput {from=MaterialExpressionMeshPartitionResource_0, output=Channel Texture}]
   properties {channel=ForestFloor, meshPartitionDefinition=&MPD_Basic}
  }
  MaterialExpressionMeshPartitionChannelSample_3 {
   class=MaterialExpressionMeshPartitionChannelSample
   inputs [ChannelTextureInput {from=MaterialExpressionMeshPartitionResource_0, output=Channel Texture}]
   properties {channel=Mud, meshPartitionDefinition=&MPD_Basic}
  }
  MaterialExpressionMeshPartitionChannelSample_4 {
   class=MaterialExpressionMeshPartitionChannelSample
   inputs [ChannelTextureInput {from=MaterialExpressionMeshPartitionResource_0, output=Channel Texture}]
   properties {channel=Gravel, meshPartitionDefinition=&MPD_Basic}
  }
  MaterialExpressionMeshPartitionChannelSample_5 {
   class=MaterialExpressionMeshPartitionChannelSample
   inputs [ChannelTextureInput {from=MaterialExpressionMeshPartitionResource_0, output=Channel Texture}]
   properties {channel=Rock, meshPartitionDefinition=&MPD_Basic}
  }
  MaterialExpressionMeshPartitionResource_0 {class=MaterialExpressionMeshPartitionResource, inputs [], properties {}}
  MaterialExpressionNamedRerouteDeclaration_0 {
   class=MaterialExpressionNamedRerouteDeclaration
   inputs [None {from=MaterialExpressionMeshPartitionChannelSample_1}]
   properties {name=Weight Dirt, nodeColor=<rgba 0.6198 0.3578 0.1907 1>}
  }
  MaterialExpressionNamedRerouteDeclaration_1 {
   class=MaterialExpressionNamedRerouteDeclaration
   inputs [None {from=MaterialExpressionMeshPartitionChannelSample_0}]
   properties {name=Weight Grass, nodeColor=<rgba 0.0315 0.7552 0.0625 1>}
  }
  MaterialExpressionNamedRerouteDeclaration_2 {
   class=MaterialExpressionNamedRerouteDeclaration
   inputs [None {from=MaterialExpressionConstant3Vector_0}]
   properties {name=TextureSize, nodeColor=<rgba 1 0 0.1641 1>}
  }
  MaterialExpressionNamedRerouteDeclaration_3 {
   class=MaterialExpressionNamedRerouteDeclaration
   inputs [None {from=MaterialExpressionMeshPartitionChannelSample_2}]
   properties {name=Weight ForestFloor, nodeColor=<rgba 0.4861 0.875 0.1504 1>}
  }
  MaterialExpressionNamedRerouteDeclaration_4 {
   class=MaterialExpressionNamedRerouteDeclaration
   inputs [None {from=MaterialExpressionMeshPartitionChannelSample_3}]
   properties {name=Weight Mud, nodeColor=<rgba 0.0938 0.0486 0.0271 1>}
  }
  MaterialExpressionNamedRerouteDeclaration_5 {
   class=MaterialExpressionNamedRerouteDeclaration
   inputs [None {from=MaterialExpressionMeshPartitionChannelSample_4}]
   properties {name=Weight Gravel, nodeColor=<rgba 0.224 0.201 0.1901 1>}
  }
  MaterialExpressionNamedRerouteDeclaration_6 {
   class=MaterialExpressionNamedRerouteDeclaration
   inputs [None {from=MaterialExpressionMeshPartitionChannelSample_5}]
   properties {name=Weight Rock, nodeColor=<rgba 0.4792 0.4741 0.4717 1>}
  }
  MaterialExpressionNamedRerouteUsage_0 {class=MaterialExpressionNamedRerouteUsage}
  MaterialExpressionNamedRerouteUsage_1 {class=MaterialExpressionNamedRerouteUsage}
  MaterialExpressionNamedRerouteUsage_10 {class=MaterialExpressionNamedRerouteUsage}
  MaterialExpressionNamedRerouteUsage_20 {class=MaterialExpressionNamedRerouteUsage}
  MaterialExpressionNamedRerouteUsage_21 {class=MaterialExpressionNamedRerouteUsage}
  MaterialExpressionNamedRerouteUsage_22 {class=MaterialExpressionNamedRerouteUsage}
  MaterialExpressionNamedRerouteUsage_3 {class=MaterialExpressionNamedRerouteUsage}
  MaterialExpressionNamedRerouteUsage_4 {class=MaterialExpressionNamedRerouteUsage}
  MaterialExpressionNamedRerouteUsage_5 {class=MaterialExpressionNamedRerouteUsage}
  MaterialExpressionNamedRerouteUsage_7 {class=MaterialExpressionNamedRerouteUsage}
  MaterialExpressionNamedRerouteUsage_8 {class=MaterialExpressionNamedRerouteUsage}
  MaterialExpressionNamedRerouteUsage_9 {class=MaterialExpressionNamedRerouteUsage}
  MaterialExpressionPinBase_1 {
   class=MaterialExpressionPinBase
   inputs []
   properties {reroutePins [A {expression=&self:MaterialGraph_0.MaterialGraphNode_Composite_3.Dirt.MaterialExpressionReroute_8}]}
  }
  MaterialExpressionPinBase_10 {
   class=MaterialExpressionPinBase
   inputs [Output {from=MaterialExpressionBlendMaterialAttributes_4, output=""}]
   properties {reroutePins [Output {expression=&self:MaterialGraph_0.MaterialGraphNode_Composite_5.Gravel.MaterialExpressionReroute_9}]}
  }
  MaterialExpressionPinBase_11 {
   class=MaterialExpressionPinBase
   inputs []
   properties {reroutePins [A {expression=&self:MaterialExpressionReroute_4}]}
  }
  MaterialExpressionPinBase_12 {
   class=MaterialExpressionPinBase
   inputs [Output {from=MaterialExpressionBlendMaterialAttributes_5, output=""}]
   properties {reroutePins [Output {expression=&self:MaterialExpressionReroute_5}]}
  }
  MaterialExpressionPinBase_2 {
   class=MaterialExpressionPinBase
   inputs []
   properties {reroutePins [A {expression=&self:MaterialGraph_0.MaterialGraphNode_Composite_1.Grass.MaterialExpressionReroute_0}]}
  }
  MaterialExpressionPinBase_3 {
   class=MaterialExpressionPinBase
   inputs [Output {from=MaterialExpressionBlendMaterialAttributes_0, output=""}]
   properties {reroutePins [Output {expression=&self:MaterialGraph_0.MaterialGraphNode_Composite_1.Grass.MaterialExpressionReroute_1}]}
  }
  MaterialExpressionPinBase_4 {
   class=MaterialExpressionPinBase
   inputs [Output {from=MaterialExpressionBlendMaterialAttributes_2, output=""}]
   properties {reroutePins [Output {expression=&self:MaterialGraph_0.MaterialGraphNode_Composite_3.Dirt.MaterialExpressionReroute_9}]}
  }
  MaterialExpressionPinBase_5 {
   class=MaterialExpressionPinBase
   inputs []
   properties {reroutePins [A {expression=&self:MaterialExpressionReroute_0}]}
  }
  MaterialExpressionPinBase_6 {
   class=MaterialExpressionPinBase
   inputs [Output {from=MaterialExpressionBlendMaterialAttributes_1, output=""}]
   properties {reroutePins [Output {expression=&self:MaterialExpressionReroute_3}]}
  }
  MaterialExpressionPinBase_7 {
   class=MaterialExpressionPinBase
   inputs []
   properties {reroutePins [A {expression=&self:MaterialGraph_0.MaterialGraphNode_Composite_4.Mud.MaterialExpressionReroute_4}]}
  }
  MaterialExpressionPinBase_8 {
   class=MaterialExpressionPinBase
   inputs [Output {from=MaterialExpressionBlendMaterialAttributes_3, output=""}]
   properties {reroutePins [Output {expression=&self:MaterialGraph_0.MaterialGraphNode_Composite_4.Mud.MaterialExpressionReroute_5}]}
  }
  MaterialExpressionPinBase_9 {
   class=MaterialExpressionPinBase
   inputs []
   properties {reroutePins [A {expression=&self:MaterialGraph_0.MaterialGraphNode_Composite_5.Gravel.MaterialExpressionReroute_8}]}
  }
  MaterialExpressionReroute_0 {class=MaterialExpressionReroute, inputs [None {from=MaterialExpressionReroute_1, output=Output}]}
  MaterialExpressionReroute_3 {class=MaterialExpressionReroute, inputs [None {from=MaterialExpressionBlendMaterialAttributes_1}]}
  MaterialExpressionReroute_4 {class=MaterialExpressionReroute, inputs [None {from=MaterialExpressionReroute_9, output=Output}]}
  MaterialExpressionReroute_5 {class=MaterialExpressionReroute, inputs [None {from=MaterialExpressionBlendMaterialAttributes_5}]}
  MaterialExpressionTextureObject_0 {class=MaterialExpressionTextureObject, inputs [], properties {texture=/PCGPrimitives_MeshPartitionInterop/Assets/Textures/Text/Text_Grass}}
  MaterialExpressionTextureObject_24 {
   class=MaterialExpressionTextureObject
   inputs []
   properties {texture=/PCGPrimitives_MeshPartitionInterop/Assets/Textures/Text/Text_ForestFloor}
  }
  MaterialExpressionTextureObject_25 {class=MaterialExpressionTextureObject, inputs [], properties {texture=/PCGPrimitives_MeshPartitionInterop/Assets/Textures/Text/Text_Dirt}}
  MaterialExpressionTextureObject_26 {class=MaterialExpressionTextureObject, inputs [], properties {texture=/PCGPrimitives_MeshPartitionInterop/Assets/Textures/Text/Text_Mud}}
  MaterialExpressionTextureObject_27 {class=MaterialExpressionTextureObject, inputs [], properties {texture=/PCGPrimitives_MeshPartitionInterop/Assets/Textures/Text/Text_Gravel}}
  MaterialExpressionTextureObject_28 {class=MaterialExpressionTextureObject, inputs [], properties {texture=/PCGPrimitives_MeshPartitionInterop/Assets/Textures/Text/Text_Rock}}
 }
 properties {
  bEnableDisplacementFade=true
  bEnableTessellation=true
  bUseMaterialAttributes=true
  bUsedWithNanite=true
  usedShadingModels=Default Lit
  displacementScaling {center=0.5, magnitude=1}
 }
}
