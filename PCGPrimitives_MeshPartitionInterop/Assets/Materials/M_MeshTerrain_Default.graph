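# Default lit, Nanite-enabled terrain material for Mesh Partition meshes that shows painted surface layers as tinted grid colors.
# A halved white grid texture, projected on world XY, serves as the untinted base and is multiplied by each layer's color parameter.
# Grass, ForestSoil, Dirt and Rock channels from MPD_Default drive chained lerps in that order, so later layers cover earlier ones.
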Material /PCGPrimitives_MeshPartitionInterop/Assets/Materials/M_MeshTerrain_Default
graph {
 properties {bUsedWithNanite=true, usedShadingModels=Default Lit}
 nodes [
  n1 = WorldPosition()
  n2 = Multiply(A=TextureSample(UVs=Multiply(A=@n1.XY) {constB=0.0025}) {texture=/Engine/EngineMaterials/DefaultWhiteGrid}.RGB) {constB=0.5}
  n3 = MeshPartitionResource()
  n4 = LinearInterpolate(
   A=@n2
   B=Multiply(A=@n2, B=VectorParameter() {parameterName=ColorGrass, defaultValue=<rgba 0.0709 0.2135 0.0283 1>}.RGB) {desc=Grass}
   Alpha=MeshPartitionChannelSample!(ChannelTextureInput=@n3.Channel Texture) {channel=Grass, meshPartitionDefinition=/PCGPrimitives_MeshPartitionInterop/Assets/MeshPartitionDefinitions/MPD_Default}
  )
  n5 = Multiply(A=@n2, B=VectorParameter() {parameterName=ColorForestSoil, defaultValue=<rgba 0.1302 0.101 0.0397 1>}.RGB)
  n6 = MeshPartitionChannelSample(ChannelTextureInput=@n3.Channel Texture) {channel=ForestSoil, meshPartitionDefinition=&MPD_Default}
  n7 = VectorParameter() {parameterName=ColorDirt, defaultValue=<rgba 0.1146 0.0586 0.0188 1>}
  LinearInterpolate(
   A=LinearInterpolate(
    A=LinearInterpolate(A=@n4, B=@n5, Alpha=@n6)
    B=Multiply(A=@n2, B=@n7.RGB)
    Alpha=MeshPartitionChannelSample(ChannelTextureInput=@n3.Channel Texture) {channel=Dirt, meshPartitionDefinition=&MPD_Default}
   )
   B=Multiply(A=@n2, B=VectorParameter() {parameterName=ColorRock, defaultValue=<rgba 0.05 0.05 0.05 0>}.RGB)
   Alpha=MeshPartitionChannelSample(ChannelTextureInput=@n3.Channel Texture) {channel=Rock, meshPartitionDefinition=&MPD_Default}
  )
 ]
}
